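; Stage disco ball building prop with optional disco effects.
; The construction script passes ceiling-attachment and floor-clipping flags into its particle system and sets the system's auto-activate from the DiscoEffects flag.
; On BeginPlay, only where cosmetic events can execute, it sets the body material's rotation speed to a small random value when effects are on, or zero when off, so the spin is material-driven.
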(event EventBeginPlay
  (Networking|CanExecuteCosmeticEvents
    (:True
      (if (Variables|Default|GetDiscoEffects)
        (Rendering|Material|SetScalarParameterValueonMaterials (Variables|Default|GetBody) "Rotation Speed" (Math|Random|RandomFloatinRange 0.05 0.1))
        (else
          (Rendering|Material|SetScalarParameterValueonMaterials (Variables|Default|GetBody) "Rotation Speed" 0.0))))))

(fn ConstructionScript
  (bind _p_discoballfx_clyde_converted (Variables|Default|GetP_DiscoBallFX_Clyde_Converted))
  (Effects|Components|ParticleSystem|SetBoolParameter (Variables|Default|GetP_DiscoBallFX_Clyde_Converted) "IsAttatchedToCeiling" (Variables|Default|GetIsAttatchedToACeiling))
  (Effects|Components|ParticleSystem|SetBoolParameter (Variables|Default|GetP_DiscoBallFX_Clyde_Converted) "IsClippingThroughFloor" (Variables|Default|GetIsClippingThroughFloor))
  (if (Variables|Default|GetDiscoEffects)
    (Components|Activation|SetAutoActivate _p_discoballfx_clyde_converted true)
    (else
      (Components|Activation|SetAutoActivate _p_discoballfx_clyde_converted false))))
